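; Base Blueprint actor, apparently for spawning vending machine AI, that triggers a Living World spawn event when play begins.
; On the authority, it waits a configurable delay, then calls TrySpawnEvent on the game state's FortAthenaLivingWorldManager using its event row handle and own transform.
; Remote instances instead activate the Niagara component, keeping the visual effect client-side.
; Overlap, Tick and the construction script are empty.
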(event EventBeginPlay
  (bind _self self)
  (bind _returnvalue (Fortnite|Athena|GameState|GetGameStateAthenafromContext _self))
  (bind _returnvalue_2 (Actor|GetComponentbyClass _returnvalue "/Script/LagerRuntime.FortAthenaLivingWorldManager"))
  (bind _niagara (Variables|Default|GetNiagara))
  (Networking|SwitchHasAuthority
    (:Authority
      (Utilities|FlowControl|Delay (Variables|Default|GetSpawnDelay))
      (Utilities|IsValid _returnvalue_2
        (:Is%20Valid
          (Class|FortAthenaLivingWorldManager|TrySpawnEvent _returnvalue_2 (Variables|Default|GetEventRowHandle) (Transformation|GetActorTransform self) 0))))
    (:Remote
      (Utilities|IsValid _niagara
        (:Is%20Valid
          (Components|Activation|Activate _niagara false))))))

(event Collision|EventActorBeginOverlap
  (params (OtherActor :type Actor%20Object%20Reference)))

(event EventTick
  (params (DeltaSeconds :type Float%20%28single-precision%29)))

(fn ConstructionScript)
